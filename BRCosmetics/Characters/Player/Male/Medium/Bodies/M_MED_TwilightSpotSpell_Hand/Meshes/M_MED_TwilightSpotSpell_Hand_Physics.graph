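# Physics asset for the TwilightSpotSpell_Hand medium male player body cosmetic, providing simulation bodies and joints for a dangling necklace.
# Capsules on spine_04 and spine_05 form the torso volume, while two small capsules and an end box form the three-bone dyn_necklace chain.
# The necklace hangs from spine_05 through successive constraints with locked twist and limited swing, tighter on one axis than the other.
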PhysicsAsset /BRCosmetics/Characters/Player/Male/Medium/Bodies/M_MED_TwilightSpotSpell_Hand/Meshes/M_MED_TwilightSpotSpell_Hand_Physics
bodies {
 dyn_necklace_01 [{shapeType=Capsule, length=2.5, radius=1.5, shapeName=dyn_necklace_01_capsule, center=<2.4944 0 1.0886>, rotation=<rot -90 0 0>}]
 dyn_necklace_02 [{shapeType=Capsule, length=2, radius=1.5, shapeName=dyn_necklace_02_capsule, center=<2.4952 0 0.4336>, rotation=<rot -90 -180 180>}]
 dyn_necklace_03 [{shapeType=Box, extentX=9, extentY=6, extentZ=4.01, shapeName=dyn_necklace_03_box, center=<5.1308 0 0.4821>, rotation=<rot -9.9999 0 0>}]
 spine_04 [{shapeType=Capsule, length=17.0823, radius=12.6977, center=<2.6412 -0.8 1.2949>, rotation=<rot -3.5258 -90 2.9451>}]
 spine_05 [{shapeType=Capsule, length=23.466, radius=10.9229, center=<1.0079 -2.0614 0.0692>, rotation=<rot 0 -131.4408 -176.1693>}]
}
constraints [
 {bone1Name=spine_05, bone2Name=spine_04, swing1LimitDegrees=45, swing2LimitDegrees=45, twistLimitDegrees=45}
 {bone1Name=dyn_necklace_02, bone2Name=dyn_necklace_01, swing1LimitDegrees=20, swing2LimitDegrees=45, twistLimitDegrees=0}
 {bone1Name=dyn_necklace_01, bone2Name=spine_05, swing1LimitDegrees=10, swing2LimitDegrees=45, twistLimitDegrees=0}
 {bone1Name=dyn_necklace_03, bone2Name=dyn_necklace_02, swing1LimitDegrees=20, swing2LimitDegrees=45, twistLimitDegrees=0}
]
