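# Anim Blueprint for the TwilightSpotSpell_Hand male body cosmetic, adding secondary motion to its pajama pants.
# Copies the parent mesh pose through a Control Rig and spine_05-based RigidBody simulation, then chains sprung AnimDynamics nodes over bottom, middle and top pant bones.
# Skydiving and front-end events set spring, alpha and simulation-space variables, but no shown node reads them.
# Weighted leg-bone Constraint nodes sit unwired from the output.
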AnimBlueprint /BRCosmetics/Characters/Player/Male/Medium/Bodies/M_MED_TwilightSpotSpell_Hand/Meshes/M_MED_TwilightSpotSpell_Hand_AnimBP
target_skeleton=&/M_MED_TwilightSpotSpell_Hand_Skeleton
graphs [
 AnimGraph {nodes [
   Constraint!(
    ComponentPose
    Alpha=1
    ConstraintWeights_0=0.0014
    ConstraintWeights_1:float=0.0144
    ConstraintWeights_2:float=0.0006
    ConstraintWeights_3:float=0.5514
    ConstraintWeights_4:float=0.0199
    ConstraintWeights_5:float=0.4112
    ConstraintWeights_6:float=0.001
   ) -> (Pose) {
    alphaBoolBlend {blendOption=Linear}
    constraintSetup [
     {targetBone {boneName=calf_twist_02_l}}
     {targetBone {boneName=thigh_r}}
     {targetBone {boneName=thigh_twist_01_r}}
     {targetBone {boneName=calf_r}}
     {targetBone {boneName=calf_twist_01_r}}
     {targetBone {boneName=calf_twist_02_r}}
     {targetBone {boneName=ball_r}}
    ]
    constraintWeights [1,1,1,1,1,1,1]
    boneToModify=~
   }
   Constraint!(
    ComponentPose
    Alpha=1
    ConstraintWeights_0=0.0108
    ConstraintWeights_1=0.5298
    ConstraintWeights_2=0.0209
    ConstraintWeights_3=0.4358
    ConstraintWeights_4=0.0014
    ConstraintWeights_5=0.0014
   ) -> (Pose) {
    alphaBoolBlend {blendOption=Linear}
    constraintSetup [
     {targetBone {boneName=thigh_l}}
     {targetBone {boneName=calf_l}}
     {targetBone {boneName=calf_twist_01_l}}
     {targetBone {boneName=calf_twist_02_l}}
     {targetBone {boneName=ball_l}}
     {targetBone {boneName=calf_twist_02_r}}
    ]
    constraintWeights [1,1,1,1,1,1]
    boneToModify=~
   }
   Constraint(
    Alpha=1
    ConstraintWeights_0=0.0009
    ConstraintWeights_1=0.3106
    ConstraintWeights_2=0.2258
    ConstraintWeights_3=0.0046
    ConstraintWeights_4=0.4251
    ConstraintWeights_5=0.0331
   ) {
    alphaBoolBlend {blendOption=Linear}
    constraintSetup [
     {targetBone {boneName=calf_twist_01_l}}
     {targetBone {boneName=calf_r}}
     {targetBone {boneName=calf_twist_01_r}}
     {targetBone {boneName=calf_twist_02_r}}
     {targetBone {boneName=foot_r}}
     {targetBone {boneName=ball_r}}
    ]
    constraintWeights [1,1,1,1,1,1]
    boneToModify=~
   }
   Constraint(
    Alpha=1
    ConstraintWeights_0=0.0042
    ConstraintWeights_1=0.606
    ConstraintWeights_2=0.2781
    ConstraintWeights_3=0.0061
    ConstraintWeights_4=0.0944
    ConstraintWeights_5=0.0113
   ) {
    alphaBoolBlend {blendOption=Linear}
    constraintSetup [
     {targetBone {boneName=calf_twist_01_l}}
     {targetBone {boneName=calf_r}}
     {targetBone {boneName=calf_twist_01_r}}
     {targetBone {boneName=calf_twist_02_r}}
     {targetBone {boneName=foot_r}}
     {targetBone {boneName=ball_r}}
    ]
    constraintWeights [1,1,1,1,1,1]
    boneToModify=~
   }
   Constraint(
    Alpha=1
    ConstraintWeights_0=0.003
    ConstraintWeights_1=0.5104
    ConstraintWeights_2=0.1229
    ConstraintWeights_3=0.003
    ConstraintWeights_4=0.3436
    ConstraintWeights_5=0.017
   ) {
    alphaBoolBlend {blendOption=Linear}
    constraintSetup [
     {targetBone {boneName=calf_twist_01_l}}
     {targetBone {boneName=calf_r}}
     {targetBone {boneName=calf_twist_01_r}}
     {targetBone {boneName=calf_twist_02_r}}
     {targetBone {boneName=foot_r}}
     {targetBone {boneName=ball_r}}
    ]
    constraintWeights [1,1,1,1,1,1]
    boneToModify=~
   }
   Constraint(
    Alpha=1
    ConstraintWeights_0=0.0019
    ConstraintWeights_1=0.5573
    ConstraintWeights_2=0.2128
    ConstraintWeights_3=0.0052
    ConstraintWeights_4=0.2145
    ConstraintWeights_5=0.0082
   ) {
    alphaBoolBlend {blendOption=Linear}
    constraintSetup [
     {targetBone {boneName=calf_twist_01_l}}
     {targetBone {boneName=calf_r}}
     {targetBone {boneName=calf_twist_01_r}}
     {targetBone {boneName=calf_twist_02_r}}
     {targetBone {boneName=foot_r}}
     {targetBone {boneName=ball_r}}
    ]
    constraintWeights [1,1,1,1,1,1]
    boneToModify=~
   }
   Constraint(
    Alpha=1
    ConstraintWeights_0=0.0004
    ConstraintWeights_1=0.5038
    ConstraintWeights_2=0.3533
    ConstraintWeights_3=0.004
    ConstraintWeights_4=0.1298
    ConstraintWeights_5=0.0087
   ) {
    alphaBoolBlend {blendOption=Linear}
    constraintSetup [
     {targetBone {boneName=calf_twist_01_l}}
     {targetBone {boneName=calf_r}}
     {targetBone {boneName=calf_twist_01_r}}
     {targetBone {boneName=calf_twist_02_r}}
     {targetBone {boneName=foot_r}}
     {targetBone {boneName=ball_r}}
    ]
    constraintWeights [1,1,1,1,1,1]
    boneToModify=~
   }
   Constraint(
    Alpha=1
    ConstraintWeights_0=0.0005
    ConstraintWeights_1=0.4302
    ConstraintWeights_2=0.139
    ConstraintWeights_3=0.0023
    ConstraintWeights_4=0.4145
    ConstraintWeights_5=0.0136
   ) {
    alphaBoolBlend {blendOption=Linear}
    constraintSetup [
     {targetBone {boneName=calf_twist_01_l}}
     {targetBone {boneName=calf_r}}
     {targetBone {boneName=calf_twist_01_r}}
     {targetBone {boneName=calf_twist_02_r}}
     {targetBone {boneName=foot_r}}
     {targetBone {boneName=ball_r}}
    ]
    constraintWeights [1,1,1,1,1,1]
    boneToModify=~
   }
   Constraint!(
    ComponentPose
    Alpha=1
    ConstraintWeights_0=0.0025
    ConstraintWeights_1=0.0087
    ConstraintWeights_2=0.0133
    ConstraintWeights_3=0.5315
    ConstraintWeights_4=0.0143
    ConstraintWeights_5=0.4294
    ConstraintWeights_6:float=0.0003
   ) -> (Pose) {
    alphaBoolBlend {blendOption=Linear}
    constraintSetup [
     {targetBone {boneName=calf_twist_01_l}}
     {targetBone {boneName=calf_twist_02_l}}
     {targetBone {boneName=thigh_r}}
     {targetBone {boneName=calf_r}}
     {targetBone {boneName=calf_twist_01_r}}
     {targetBone {boneName=calf_twist_02_r}}
     {targetBone {boneName=foot_r}}
    ]
    constraintWeights [1,1,1,1,1,1,1]
    boneToModify=~
   }
   Constraint!(
    ComponentPose
    Alpha=1
    ConstraintWeights_0=0.3868
    ConstraintWeights_1=0.1389
    ConstraintWeights_2=0.002
    ConstraintWeights_3=0.4584
    ConstraintWeights_4=0.0134
    ConstraintWeights_5=0.0004
   ) -> (Pose) {
    alphaBoolBlend {blendOption=Linear}
    constraintSetup [
     {targetBone {boneName=calf_l}}
     {targetBone {boneName=calf_twist_01_l}}
     {targetBone {boneName=calf_twist_02_l}}
     {targetBone {boneName=foot_l}}
     {targetBone {boneName=ball_l}}
     {targetBone {boneName=calf_twist_01_r}}
    ]
    constraintWeights [1,1,1,1,1,1]
    boneToModify=~
   }
   Constraint(
    Alpha=1
    ConstraintWeights_0=0.5085
    ConstraintWeights_1=0.3615
    ConstraintWeights_2=0.0036
    ConstraintWeights_3=0.115
    ConstraintWeights_4=0.0109
    ConstraintWeights_5=0.0004
   ) {
    alphaBoolBlend {blendOption=Linear}
    constraintSetup [
     {targetBone {boneName=calf_l}}
     {targetBone {boneName=calf_twist_01_l}}
     {targetBone {boneName=calf_twist_02_l}}
     {targetBone {boneName=foot_l}}
     {targetBone {boneName=ball_l}}
     {targetBone {boneName=calf_twist_01_r}}
    ]
    constraintWeights [1,1,1,1,1,1]
    boneToModify=~
   }
   Constraint(
    Alpha=1
    ConstraintWeights_0=0.3562
    ConstraintWeights_1=0.2649
    ConstraintWeights_2=0.0056
    ConstraintWeights_3=0.3329
    ConstraintWeights_4=0.0392
    ConstraintWeights_5=0.0011
   ) {
    alphaBoolBlend {blendOption=Linear}
    constraintSetup [
     {targetBone {boneName=calf_l}}
     {targetBone {boneName=calf_twist_01_l}}
     {targetBone {boneName=calf_twist_02_l}}
     {targetBone {boneName=foot_l}}
     {targetBone {boneName=ball_l}}
     {targetBone {boneName=calf_twist_01_r}}
    ]
    constraintWeights [1,1,1,1,1,1]
    boneToModify=~
   }
   Constraint!(
    ComponentPose
    Alpha=1
    ConstraintWeights_0=0.5864
    ConstraintWeights_1=0.3153
    ConstraintWeights_2=0.0072
    ConstraintWeights_3=0.0772
    ConstraintWeights_4=0.0093
    ConstraintWeights_5=0.004
    ConstraintWeights_6:float=0.0005
   ) -> (Pose) {
    alphaBoolBlend {blendOption=Linear}
    constraintSetup [
     {targetBone {boneName=calf_l}}
     {targetBone {boneName=calf_twist_01_l}}
     {targetBone {boneName=calf_twist_02_l}}
     {targetBone {boneName=foot_l}}
     {targetBone {boneName=ball_l}}
     {targetBone {boneName=calf_twist_01_r}}
     {targetBone {boneName=calf_twist_02_r}}
    ]
    constraintWeights [1,1,1,1,1,1,1]
    boneToModify=~
   }
   Constraint!(
    ComponentPose
    Alpha=1
    ConstraintWeights_0=0.5083
    ConstraintWeights_1=0.118
    ConstraintWeights_2=0.0029
    ConstraintWeights_3=0.3508
    ConstraintWeights_4=0.0171
    ConstraintWeights_5=0.0029
   ) -> (Pose) {
    alphaBoolBlend {blendOption=Linear}
    constraintSetup [
     {targetBone {boneName=calf_l}}
     {targetBone {boneName=calf_twist_01_l}}
     {targetBone {boneName=calf_twist_02_l}}
     {targetBone {boneName=foot_l}}
     {targetBone {boneName=ball_l}}
     {targetBone {boneName=calf_twist_01_r}}
    ]
    constraintWeights [1,1,1,1,1,1]
    boneToModify=~
   }
   Constraint(
    Alpha=1
    ConstraintWeights_0=0.5649
    ConstraintWeights_1=0.2018
    ConstraintWeights_2=0.0049
    ConstraintWeights_3=0.218
    ConstraintWeights_4=0.0083
    ConstraintWeights_5=0.002
   ) {
    alphaBoolBlend {blendOption=Linear}
    constraintSetup [
     {targetBone {boneName=calf_l}}
     {targetBone {boneName=calf_twist_01_l}}
     {targetBone {boneName=calf_twist_02_l}}
     {targetBone {boneName=foot_l}}
     {targetBone {boneName=ball_l}}
     {targetBone {boneName=calf_twist_01_r}}
    ]
    constraintWeights [1,1,1,1,1,1]
    boneToModify=~
   }
   Constraint!(
    ComponentPose
    Alpha=1
    ConstraintWeights_0=0.0137
    ConstraintWeights_1=0.5148
    ConstraintWeights_2=0.0101
    ConstraintWeights_3=0.4493
    ConstraintWeights_4=0.0013
    ConstraintWeights_5=0.0015
    ConstraintWeights_6:float=0.0093
   ) -> (Pose) {
    alphaBoolBlend {blendOption=Linear}
    constraintSetup [
     {targetBone {boneName=thigh_l}}
     {targetBone {boneName=calf_l}}
     {targetBone {boneName=calf_twist_01_l}}
     {targetBone {boneName=calf_twist_02_l}}
     {targetBone {boneName=foot_l}}
     {targetBone {boneName=calf_twist_01_r}}
     {targetBone {boneName=calf_twist_02_r}}
    ]
    constraintWeights [1,1,1,1,1,1,1]
    boneToModify=~
   }
   n1 = Anim Dynamics - Bone: dyn_pjs_bot_l_A!(
    SimulationSpace:AnimPhysSimSpaceType Enum
    ComponentPose=Local To Component(LocalPose=Use cached pose 'RBANCache'())
    Alpha=1
    LinearSpringConstant:float=30
    ExternalForce="0,0,0"
   ) -> (Pose) {
    angularDampingOverride=0.9999
    angularSpringConstant=2000
    bAngularSpring=true
    bGravityOverrideInSimSpace=true
    bLinearSpring=true
    bOverrideAngularDamping=true
    bOverrideLinearDamping=true
    bUseGravityOverride=true
    linearDampingOverride=0.9999
    linearSpringConstant=30
    simulationSpace=RootRelative
    alphaBoolBlend {blendOption=Linear}
    boundBone {boneName=dyn_pjs_bot_l_A}
    componentAppliedLinearAccClamp=<1000 1000 1000>
    componentLinearAccScale=<1 1 1>
    componentLinearVelScale=<1 1 1>
    gravityOverride {z=-980}
    physicsBodyDefinitions [{
      boundBone {boneName=dyn_pjs_bot_l_A}
      boxExtents=<7 7 7>
      constraintSetup {angularLimitsMax=<5 5 5>, angularLimitsMin=<-5 -10 -5>, angularTarget=<0 0 0>, linearAxesMax=<2 2 4>, linearAxesMin=<-2 -2 0>}
      localJointOffset=<0.1 0 0>
     }]
   }
   n2 = Anim Dynamics - Bone: dyn_pjs_bot_l_B!(SimulationSpace:AnimPhysSimSpaceType Enum, ComponentPose=@n1, Alpha=1, LinearSpringConstant:float=30, ExternalForce="0,0,0") -> (Pose) {
    angularDampingOverride=0.9999
    angularSpringConstant=2000
    bAngularSpring=true
    bGravityOverrideInSimSpace=true
    bLinearSpring=true
    bOverrideAngularDamping=true
    bOverrideLinearDamping=true
    bUseGravityOverride=true
    linearDampingOverride=0.9999
    linearSpringConstant=30
    simulationSpace=RootRelative
    alphaBoolBlend {blendOption=Linear}
    boundBone {boneName=dyn_pjs_bot_l_B}
    componentAppliedLinearAccClamp=<1000 1000 1000>
    componentLinearAccScale=<1 1 1>
    componentLinearVelScale=<1 1 1>
    gravityOverride {z=-980}
    physicsBodyDefinitions [{
      boundBone {boneName=dyn_pjs_bot_l_B}
      boxExtents=<7 7 7>
      constraintSetup {angularLimitsMax=<5 5 5>, angularLimitsMin=<-5 -10 -5>, angularTarget=<0 0 0>, linearAxesMax=<2 2 4>, linearAxesMin=<-2 -2 0>}
      localJointOffset=<0.1 0 0>
     }]
   }
   n3 = Anim Dynamics - Bone: dyn_pjs_bot_l_C!(SimulationSpace:AnimPhysSimSpaceType Enum, ComponentPose=@n2, Alpha=1, LinearSpringConstant:float=30, ExternalForce="0,0,0") -> (Pose) {
    angularDampingOverride=0.9999
    angularSpringConstant=2000
    bAngularSpring=true
    bGravityOverrideInSimSpace=true
    bLinearSpring=true
    bOverrideAngularDamping=true
    bOverrideLinearDamping=true
    bUseGravityOverride=true
    linearDampingOverride=0.9999
    linearSpringConstant=30
    simulationSpace=RootRelative
    alphaBoolBlend {blendOption=Linear}
    boundBone {boneName=dyn_pjs_bot_l_C}
    componentAppliedLinearAccClamp=<1000 1000 1000>
    componentLinearAccScale=<1 1 1>
    componentLinearVelScale=<1 1 1>
    gravityOverride {z=-980}
    physicsBodyDefinitions [{
      boundBone {boneName=dyn_pjs_bot_l_C}
      boxExtents=<7 7 7>
      constraintSetup {angularLimitsMax=<5 5 5>, angularLimitsMin=<-5 -10 -5>, angularTarget=<0 0 0>, linearAxesMax=<2 2 4>, linearAxesMin=<-2 -2 0>}
      localJointOffset=<0.1 0 0>
     }]
   }
   n4 = Anim Dynamics - Bone: dyn_pjs_bot_l_D!(SimulationSpace:AnimPhysSimSpaceType Enum, ComponentPose=@n3, Alpha=1, LinearSpringConstant:float=30, ExternalForce="0,0,0") -> (Pose) {
    angularDampingOverride=0.9999
    angularSpringConstant=2000
    bAngularSpring=true
    bGravityOverrideInSimSpace=true
    bLinearSpring=true
    bOverrideAngularDamping=true
    bOverrideLinearDamping=true
    bUseGravityOverride=true
    linearDampingOverride=0.9999
    linearSpringConstant=30
    simulationSpace=RootRelative
    alphaBoolBlend {blendOption=Linear}
    boundBone {boneName=dyn_pjs_bot_l_D}
    componentAppliedLinearAccClamp=<1000 1000 1000>
    componentLinearAccScale=<1 1 1>
    componentLinearVelScale=<1 1 1>
    gravityOverride {z=-980}
    physicsBodyDefinitions [{
      boundBone {boneName=dyn_pjs_bot_l_D}
      boxExtents=<7 7 7>
      constraintSetup {angularLimitsMax=<5 5 5>, angularLimitsMin=<-5 -10 -5>, angularTarget=<0 0 0>, linearAxesMax=<2 2 4>, linearAxesMin=<-2 -2 0>}
      localJointOffset=<0.1 0 0>
     }]
   }
   n5 = Anim Dynamics - Bone: dyn_pjs_bot_l_E!(SimulationSpace:AnimPhysSimSpaceType Enum, ComponentPose=@n4, Alpha=1, LinearSpringConstant:float=30, ExternalForce="0,0,0") -> (Pose) {
    angularDampingOverride=0.9999
    angularSpringConstant=2000
    bAngularSpring=true
    bGravityOverrideInSimSpace=true
    bLinearSpring=true
    bOverrideAngularDamping=true
    bOverrideLinearDamping=true
    bUseGravityOverride=true
    linearDampingOverride=0.9999
    linearSpringConstant=30
    simulationSpace=RootRelative
    alphaBoolBlend {blendOption=Linear}
    boundBone {boneName=dyn_pjs_bot_l_E}
    componentAppliedLinearAccClamp=<1000 1000 1000>
    componentLinearAccScale=<1 1 1>
    componentLinearVelScale=<1 1 1>
    gravityOverride {z=-980}
    physicsBodyDefinitions [{
      boundBone {boneName=dyn_pjs_bot_l_E}
      boxExtents=<7 7 7>
      constraintSetup {angularLimitsMax=<5 5 5>, angularLimitsMin=<-5 -10 -5>, angularTarget=<0 0 0>, linearAxesMax=<2 2 5>, linearAxesMin=<-2 -2 0>}
      localJointOffset=<0.1 0 0>
     }]
   }
   n6 = Anim Dynamics - Bone: dyn_pjs_bot_l_F!(SimulationSpace:AnimPhysSimSpaceType Enum, ComponentPose=@n5, Alpha=1, LinearSpringConstant:float=30, ExternalForce="0,0,0") -> (Pose) {
    angularDampingOverride=0.9999
    angularSpringConstant=2000
    bAngularSpring=true
    bGravityOverrideInSimSpace=true
    bLinearSpring=true
    bOverrideAngularDamping=true
    bOverrideLinearDamping=true
    bUseGravityOverride=true
    linearDampingOverride=0.9999
    linearSpringConstant=30
    simulationSpace=RootRelative
    alphaBoolBlend {blendOption=Linear}
    boundBone {boneName=dyn_pjs_bot_l_F}
    componentAppliedLinearAccClamp=<1000 1000 1000>
    componentLinearAccScale=<1 1 1>
    componentLinearVelScale=<1 1 1>
    gravityOverride {z=-980}
    physicsBodyDefinitions [{
      boundBone {boneName=dyn_pjs_bot_l_F}
      boxExtents=<7 7 7>
      constraintSetup {angularLimitsMax=<5 5 5>, angularLimitsMin=<-5 -10 -5>, angularTarget=<0 0 0>, linearAxesMax=<2 2 4>, linearAxesMin=<-2 -2 0>}
      localJointOffset=<0.1 0 0>
     }]
   }
   n7 = Anim Dynamics - Bone: dyn_pjs_bot_r_A!(SimulationSpace:AnimPhysSimSpaceType Enum, ComponentPose=@n6, Alpha=1, LinearSpringConstant:float=30, ExternalForce="0,0,0") -> (Pose) {
    angularDampingOverride=0.9999
    angularSpringConstant=2000
    bAngularSpring=true
    bGravityOverrideInSimSpace=true
    bLinearSpring=true
    bOverrideAngularDamping=true
    bOverrideLinearDamping=true
    bUseGravityOverride=true
    linearDampingOverride=0.9999
    linearSpringConstant=30
    simulationSpace=RootRelative
    alphaBoolBlend {blendOption=Linear}
    boundBone {boneName=dyn_pjs_bot_r_A}
    componentAppliedLinearAccClamp=<1000 1000 1000>
    componentLinearAccScale=<1 1 1>
    componentLinearVelScale=<1 1 1>
    gravityOverride {z=-980}
    physicsBodyDefinitions [{
      boundBone {boneName=dyn_pjs_bot_r_A}
      boxExtents=<7 7 7>
      constraintSetup {angularLimitsMax=<5 5 5>, angularLimitsMin=<-5 -10 -5>, angularTarget=<0 0 0>, linearAxesMax=<2 2 4>, linearAxesMin=<-2 -2 0>}
      localJointOffset=<0.1 0 0>
     }]
   }
   n8 = Anim Dynamics - Bone: dyn_pjs_bot_r_B!(SimulationSpace:AnimPhysSimSpaceType Enum, ComponentPose=@n7, Alpha=1, LinearSpringConstant:float=30, ExternalForce="0,0,0") -> (Pose) {
    angularDampingOverride=0.9999
    angularSpringConstant=2000
    bAngularSpring=true
    bGravityOverrideInSimSpace=true
    bLinearSpring=true
    bOverrideAngularDamping=true
    bOverrideLinearDamping=true
    bUseGravityOverride=true
    linearDampingOverride=0.9999
    linearSpringConstant=30
    simulationSpace=RootRelative
    alphaBoolBlend {blendOption=Linear}
    boundBone {boneName=dyn_pjs_bot_r_B}
    componentAppliedLinearAccClamp=<1000 1000 1000>
    componentLinearAccScale=<1 1 1>
    componentLinearVelScale=<1 1 1>
    gravityOverride {z=-980}
    physicsBodyDefinitions [{
      boundBone {boneName=dyn_pjs_bot_r_B}
      boxExtents=<7 7 7>
      constraintSetup {angularLimitsMax=<5 5 5>, angularLimitsMin=<-5 -10 -5>, angularTarget=<0 0 0>, linearAxesMax=<2 2 4>, linearAxesMin=<-2 -2 0>}
      localJointOffset=<0.1 0 0>
     }]
   }
   n9 = Anim Dynamics - Bone: dyn_pjs_bot_r_C!(SimulationSpace:AnimPhysSimSpaceType Enum, ComponentPose=@n8, Alpha=1, LinearSpringConstant:float=30, ExternalForce="0,0,0") -> (Pose) {
    angularDampingOverride=0.9999
    angularSpringConstant=2000
    bAngularSpring=true
    bGravityOverrideInSimSpace=true
    bLinearSpring=true
    bOverrideAngularDamping=true
    bOverrideLinearDamping=true
    bUseGravityOverride=true
    linearDampingOverride=0.9999
    linearSpringConstant=30
    simulationSpace=RootRelative
    alphaBoolBlend {blendOption=Linear}
    boundBone {boneName=dyn_pjs_bot_r_C}
    componentAppliedLinearAccClamp=<1000 1000 1000>
    componentLinearAccScale=<1 1 1>
    componentLinearVelScale=<1 1 1>
    gravityOverride {z=-980}
    physicsBodyDefinitions [{
      boundBone {boneName=dyn_pjs_bot_r_C}
      boxExtents=<7 7 7>
      constraintSetup {angularLimitsMax=<5 5 5>, angularLimitsMin=<-5 -10 -5>, angularTarget=<0 0 0>, linearAxesMax=<2 2 4>, linearAxesMin=<-2 -2 0>}
      localJointOffset=<0.1 0 0>
     }]
   }
   n10 = Anim Dynamics - Bone: dyn_pjs_bot_r_D!(SimulationSpace:AnimPhysSimSpaceType Enum, ComponentPose=@n9, Alpha=1, LinearSpringConstant:float=30, ExternalForce="0,0,0") -> (Pose) {
    angularDampingOverride=0.9999
    angularSpringConstant=2000
    bAngularSpring=true
    bGravityOverrideInSimSpace=true
    bLinearSpring=true
    bOverrideAngularDamping=true
    bOverrideLinearDamping=true
    bUseGravityOverride=true
    linearDampingOverride=0.9999
    linearSpringConstant=30
    simulationSpace=RootRelative
    alphaBoolBlend {blendOption=Linear}
    boundBone {boneName=dyn_pjs_bot_r_D}
    componentAppliedLinearAccClamp=<1000 1000 1000>
    componentLinearAccScale=<1 1 1>
    componentLinearVelScale=<1 1 1>
    gravityOverride {z=-980}
    physicsBodyDefinitions [{
      boundBone {boneName=dyn_pjs_bot_r_D}
      boxExtents=<7 7 7>
      constraintSetup {angularLimitsMax=<5 5 5>, angularLimitsMin=<-5 -10 -5>, angularTarget=<0 0 0>, linearAxesMax=<2 2 4>, linearAxesMin=<-2 -2 0>}
      localJointOffset=<0.1 0 0>
     }]
   }
   n11 = Anim Dynamics - Bone: dyn_pjs_bot_r_E!(SimulationSpace:AnimPhysSimSpaceType Enum, ComponentPose=@n10, Alpha=1, LinearSpringConstant:float=30, ExternalForce="0,0,0") -> (Pose) {
    angularDampingOverride=0.9999
    angularSpringConstant=2000
    bAngularSpring=true
    bGravityOverrideInSimSpace=true
    bLinearSpring=true
    bOverrideAngularDamping=true
    bOverrideLinearDamping=true
    bUseGravityOverride=true
    linearDampingOverride=0.9999
    linearSpringConstant=30
    simulationSpace=RootRelative
    alphaBoolBlend {blendOption=Linear}
    boundBone {boneName=dyn_pjs_bot_r_E}
    componentAppliedLinearAccClamp=<1000 1000 1000>
    componentLinearAccScale=<1 1 1>
    componentLinearVelScale=<1 1 1>
    gravityOverride {z=-980}
    physicsBodyDefinitions [{
      boundBone {boneName=dyn_pjs_bot_r_E}
      boxExtents=<7 7 7>
      constraintSetup {angularLimitsMax=<5 5 5>, angularLimitsMin=<-5 -10 -5>, angularTarget=<0 0 0>, linearAxesMax=<2 2 4>, linearAxesMin=<-2 -2 0>}
      localJointOffset=<0.1 0 0>
     }]
   }
   n12 = Anim Dynamics - Bone: dyn_pjs_bot_r_F!(SimulationSpace:AnimPhysSimSpaceType Enum, ComponentPose=@n11, Alpha=1, LinearSpringConstant:float=30, ExternalForce="0,0,0") -> (Pose) {
    angularDampingOverride=0.9999
    angularSpringConstant=2000
    bAngularSpring=true
    bGravityOverrideInSimSpace=true
    bLinearSpring=true
    bOverrideAngularDamping=true
    bOverrideLinearDamping=true
    bUseGravityOverride=true
    linearDampingOverride=0.9999
    linearSpringConstant=30
    simulationSpace=RootRelative
    alphaBoolBlend {blendOption=Linear}
    boundBone {boneName=dyn_pjs_bot_r_F}
    componentAppliedLinearAccClamp=<1000 1000 1000>
    componentLinearAccScale=<1 1 1>
    componentLinearVelScale=<1 1 1>
    gravityOverride {z=-980}
    physicsBodyDefinitions [{
      boundBone {boneName=dyn_pjs_bot_r_F}
      boxExtents=<7 7 7>
      constraintSetup {angularLimitsMax=<5 5 5>, angularLimitsMin=<-5 -10 -5>, angularTarget=<0 0 0>, linearAxesMax=<2 2 4>, linearAxesMin=<-2 -2 0>}
      localJointOffset=<0.1 0 0>
     }]
   }
   n13 = Anim Dynamics - Bone: dyn_pjs_mid_inner_r!(SimulationSpace:AnimPhysSimSpaceType Enum, ComponentPose=@n12, Alpha=1, LinearSpringConstant:float=65, ExternalForce="0,0,0") -> (Pose) {
    angularDampingOverride=0.9999
    angularSpringConstant=2000
    bAngularSpring=true
    bGravityOverrideInSimSpace=true
    bLinearSpring=true
    bOverrideAngularDamping=true
    bOverrideLinearDamping=true
    bUseGravityOverride=true
    linearDampingOverride=0.9999
    linearSpringConstant=65
    simulationSpace=RootRelative
    alphaBoolBlend {blendOption=Linear}
    boundBone {boneName=dyn_pjs_mid_inner_r}
    componentAppliedLinearAccClamp=<1000 1000 1000>
    componentLinearAccScale=<1 1 1>
    componentLinearVelScale=<1 1 1>
    gravityOverride {z=-980}
    physicsBodyDefinitions [{
      boundBone {boneName=dyn_pjs_mid_inner_r}
      boxExtents=<19 19 21>
      constraintSetup {angularLimitsMax=<5 2 5>, angularLimitsMin=<-5 -5 -5>, angularTarget=<0 0 0>, linearAxesMax=<2 0 6>, linearAxesMin=<-1 0 -6>}
      localJointOffset=<0.1 0 0>
     }]
   }
   n14 = Anim Dynamics - Bone: dyn_pjs_mid_outer_r!(SimulationSpace:AnimPhysSimSpaceType Enum, ComponentPose=@n13, Alpha=1, LinearSpringConstant:float=65, ExternalForce="0,0,0") -> (Pose) {
    angularDampingOverride=0.9999
    angularSpringConstant=2000
    bAngularSpring=true
    bGravityOverrideInSimSpace=true
    bLinearSpring=true
    bOverrideAngularDamping=true
    bOverrideLinearDamping=true
    bUseGravityOverride=true
    linearDampingOverride=0.9999
    linearSpringConstant=65
    simulationSpace=RootRelative
    alphaBoolBlend {blendOption=Linear}
    boundBone {boneName=dyn_pjs_mid_outer_r}
    componentAppliedLinearAccClamp=<1000 1000 1000>
    componentLinearAccScale=<1 1 1>
    componentLinearVelScale=<1 1 1>
    gravityOverride {z=-980}
    physicsBodyDefinitions [{
      boundBone {boneName=dyn_pjs_mid_outer_r}
      boxExtents=<19 19 21>
      constraintSetup {angularLimitsMax=<5 2 5>, angularLimitsMin=<-5 -5 -5>, angularTarget=<0 0 0>, linearAxesMax=<4 0 6>, linearAxesMin=<-1 0 -6>}
      localJointOffset=<0.1 0 0>
     }]
   }
   n15 = Anim Dynamics - Bone: dyn_pjs_mid_inner_l!(SimulationSpace:AnimPhysSimSpaceType Enum, ComponentPose=@n14, Alpha=1, LinearSpringConstant:float=65, ExternalForce="0,0,0") -> (Pose) {
    angularDampingOverride=0.9999
    angularSpringConstant=2000
    bAngularSpring=true
    bGravityOverrideInSimSpace=true
    bLinearSpring=true
    bOverrideAngularDamping=true
    bOverrideLinearDamping=true
    bUseGravityOverride=true
    linearDampingOverride=0.9999
    linearSpringConstant=65
    simulationSpace=RootRelative
    alphaBoolBlend {blendOption=Linear}
    boundBone {boneName=dyn_pjs_mid_inner_l}
    componentAppliedLinearAccClamp=<1000 1000 1000>
    componentLinearAccScale=<1 1 1>
    componentLinearVelScale=<1 1 1>
    gravityOverride {z=-980}
    physicsBodyDefinitions [{
      boundBone {boneName=dyn_pjs_mid_inner_l}
      boxExtents=<18 18 20>
      constraintSetup {angularLimitsMax=<5 2 5>, angularLimitsMin=<-5 -5 -5>, angularTarget=<0 0 0>, linearAxesMax=<2 0 6>, linearAxesMin=<-1 0 -6>}
      localJointOffset=<0.1 0 0>
     }]
   }
   n16 = Anim Dynamics - Bone: dyn_pjs_mid_outer_l!(SimulationSpace:AnimPhysSimSpaceType Enum, ComponentPose=@n15, Alpha=1, LinearSpringConstant:float=65, ExternalForce="0,0,0") -> (Pose) {
    angularDampingOverride=0.9999
    angularSpringConstant=2000
    bAngularSpring=true
    bGravityOverrideInSimSpace=true
    bLinearSpring=true
    bOverrideAngularDamping=true
    bOverrideLinearDamping=true
    bUseGravityOverride=true
    linearDampingOverride=0.9999
    linearSpringConstant=65
    simulationSpace=RootRelative
    alphaBoolBlend {blendOption=Linear}
    boundBone {boneName=dyn_pjs_mid_outer_l}
    componentAppliedLinearAccClamp=<1000 1000 1000>
    componentLinearAccScale=<1 1 1>
    componentLinearVelScale=<1 1 1>
    gravityOverride {z=-980}
    physicsBodyDefinitions [{
      boundBone {boneName=dyn_pjs_mid_outer_l}
      boxExtents=<18 18 20>
      constraintSetup {angularLimitsMax=<5 2 5>, angularLimitsMin=<-5 -5 -5>, angularTarget=<0 0 0>, linearAxesMax=<4 0 6>, linearAxesMin=<-1 0 -6>}
      localJointOffset=<0.1 0 0>
     }]
   }
   n17 = Anim Dynamics - Bone: dyn_pjs_top_inner_r!(SimulationSpace:AnimPhysSimSpaceType Enum, ComponentPose=@n16, Alpha=1, ExternalForce="0,0,0") -> (Pose) {
    angularDampingOverride=0.9999
    angularSpringConstant=1000
    bAngularSpring=true
    bGravityOverrideInSimSpace=true
    bLinearSpring=true
    bOverrideAngularDamping=true
    bOverrideLinearDamping=true
    bUseGravityOverride=true
    linearDampingOverride=0.999
    linearSpringConstant=70
    simulationSpace=RootRelative
    alphaBoolBlend {blendOption=Linear}
    boundBone {boneName=dyn_pjs_top_inner_r}
    componentAppliedLinearAccClamp=<1000 1000 1000>
    componentLinearAccScale=<1 1 1>
    componentLinearVelScale=<1 1 1>
    gravityOverride {z=-980}
    physicsBodyDefinitions [{
      boundBone {boneName=dyn_pjs_top_inner_r}
      boxExtents=<13 13 13>
      constraintSetup {angularLimitsMax=<5 0 5>, angularLimitsMin=<-5 0 -5>, angularTarget=<0 0 0>, linearAxesMax=<0 2 3>, linearAxesMin=<0 -2 -2>}
      localJointOffset=<0.1 0 0>
     }]
   }
   n18 = Anim Dynamics - Bone: dyn_pjs_top_outer_r!(SimulationSpace:AnimPhysSimSpaceType Enum, ComponentPose=@n17, Alpha=1, ExternalForce="0,0,0") -> (Pose) {
    angularDampingOverride=0.9999
    angularSpringConstant=1000
    bAngularSpring=true
    bGravityOverrideInSimSpace=true
    bLinearSpring=true
    bOverrideAngularDamping=true
    bOverrideLinearDamping=true
    bUseGravityOverride=true
    linearDampingOverride=0.999
    linearSpringConstant=70
    simulationSpace=RootRelative
    alphaBoolBlend {blendOption=Linear}
    boundBone {boneName=dyn_pjs_top_outer_r}
    componentAppliedLinearAccClamp=<1000 1000 1000>
    componentLinearAccScale=<1 1 1>
    componentLinearVelScale=<1 1 1>
    gravityOverride {z=-980}
    physicsBodyDefinitions [{
      boundBone {boneName=dyn_pjs_top_outer_r}
      boxExtents=<18 18 18>
      constraintSetup {angularLimitsMax=<5 5 5>, angularLimitsMin=<-5 -5 -5>, angularTarget=<0 0 0>, linearAxesMax=<2 4 5>, linearAxesMin=<-1 -4 -2>}
      localJointOffset=<0.1 0 0>
     }]
   }
   n19 = Anim Dynamics - Bone: dyn_pjs_top_inner_l!(SimulationSpace:AnimPhysSimSpaceType Enum, ComponentPose=@n18, Alpha=1, ExternalForce="0,0,0") -> (Pose) {
    angularDampingOverride=0.9999
    angularSpringConstant=1000
    bAngularSpring=true
    bGravityOverrideInSimSpace=true
    bLinearSpring=true
    bOverrideAngularDamping=true
    bOverrideLinearDamping=true
    bUseGravityOverride=true
    linearDampingOverride=0.999
    linearSpringConstant=70
    simulationSpace=RootRelative
    alphaBoolBlend {blendOption=Linear}
    boundBone {boneName=dyn_pjs_top_inner_l}
    componentAppliedLinearAccClamp=<1000 1000 1000>
    componentLinearAccScale=<1 1 1>
    componentLinearVelScale=<1 1 1>
    gravityOverride {z=-980}
    physicsBodyDefinitions [{
      boundBone {boneName=dyn_pjs_top_inner_l}
      boxExtents=<11 11 11>
      constraintSetup {angularLimitsMax=<5 0 5>, angularLimitsMin=<-5 0 -5>, angularTarget=<0 0 0>, linearAxesMax=<1 3 4>, linearAxesMin=<-1 -3 -1>}
      localJointOffset=<0.1 0 0>
     }]
   }
   n20 = Anim Dynamics - Bone: dyn_pjs_top_outer_l!(SimulationSpace:AnimPhysSimSpaceType Enum, ComponentPose=@n19, Alpha=1, ExternalForce="0,0,0") -> (Pose) {
    angularDampingOverride=0.9999
    angularSpringConstant=1000
    bAngularSpring=true
    bGravityOverrideInSimSpace=true
    bLinearSpring=true
    bOverrideAngularDamping=true
    bOverrideLinearDamping=true
    bUseGravityOverride=true
    linearDampingOverride=0.999
    linearSpringConstant=70
    simulationSpace=RootRelative
    alphaBoolBlend {blendOption=Linear}
    boundBone {boneName=dyn_pjs_top_outer_l}
    componentAppliedLinearAccClamp=<1000 1000 1000>
    componentLinearAccScale=<1 1 1>
    componentLinearVelScale=<1 1 1>
    gravityOverride {z=-980}
    physicsBodyDefinitions [{
      boundBone {boneName=dyn_pjs_top_outer_l}
      boxExtents=<11 11 11>
      constraintSetup {angularLimitsMax=<5 5 5>, angularLimitsMin=<-5 -5 -5>, angularTarget=<0 0 0>, linearAxesMax=<1 2 4>, linearAxesMin=<-1 -2 -2>}
      localJointOffset=<0.1 0 0>
     }]
   }
   Output Pose(Result=Component To Local(ComponentPose=@n20)) {layerGroup=None}
   n21 = Control Rig(Source=Copy Pose From Mesh() {bUseAttachedParent=true}) {
    alphaBoolBlend {blendOption=Linear}
    controlRigAssetReference {blueprintRigClass=&/M_MED_TwilightSpotSpell_Hand_CtrlRig.M_MED_TwilightSpotSpell_Hand_CtrlRig_C}
   }
   Save cached pose 'SetupCache'(Pose=@n21) {cacheName=SetupCache}
   n22 = RigidBody(ComponentPose=Local To Component(LocalPose=Use cached pose 'SetupCache'())) {
    bClampLinearTranslationLimitToRefPose=true
    bOverrideWorldGravity=true
    overridePhysicsAsset=&/M_MED_TwilightSpotSpell_Hand_Physics
    simulationSpace=BaseBoneSpace
    alphaBoolBlend {blendOption=Linear}
    baseBoneRef {boneName=spine_05}
    componentAppliedLinearAccClamp=<3000 3000 3000>
    componentLinearAccScale=<0.2 0.2 0.2>
    componentLinearVelScale=<0.2 0.2 0.2>
    overrideWorldGravity {z=-980}
    simSpaceSettings {
     maxAngularAcceleration=1000
     maxAngularVelocity=1000
     maxLinearAcceleration=1000
     maxLinearVelocity=1000
     externalAngularVelocity=<0.1 0.1 0.1>
    }
   }
   Save cached pose 'RBANCache'(Pose=Component To Local(ComponentPose=@n22)) {cacheName=RBANCache}
  ]}
 EventGraph {nodes [
   Event BlueprintUpdateAnimation()
   > n1 = Branch(Condition=Get bIsSkyDiving())
   > Set PantBottomDynSpringConstant(PantBottomDynSpringConstant:double=2000) -> (Output_Get:double)
   > Set PantMidDynSpringConstant(PantMidDynSpringConstant:double=1000) -> (Output_Get:double)
   Set PantBottomDynSpringConstant(execute=@n1.else, PantBottomDynSpringConstant=50)
   > Set PantMidDynSpringConstant(PantMidDynSpringConstant=100)
   Event OnCharacterCustomizationLoaded()
   > n2 = Branch(Condition=Get bIsInFrontEnd())
   > Set FrontendAlpha(FrontendAlpha:double=0.3) -> (Output_Get:double)
   > Set Simulation Space(Simulation Space:AnimPhysSimSpaceType Enum=World) -> (Output_Get:AnimPhysSimSpaceType Enum)
   Set FrontendAlpha(execute=@n2.else, FrontendAlpha=1)
   > Set Simulation Space()
   Get bIsSkydiveDiveMode()
   Get bIsSkydiveFloating()
  ]}
]
variables [
 FrontendAlpha {type=number}
 PantBottomDynSpringConstant {type=number}
 PantMidDynSpringConstant {type=number}
 Simulation Space {type=AnimPhysSimSpaceType, default=BoneRelative}
]
